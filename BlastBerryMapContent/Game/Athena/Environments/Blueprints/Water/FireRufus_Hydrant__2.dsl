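; Fire hydrant building actor that, when damaged, switches to an Exploded state and becomes a water-jet launch pad.
; The custom state drives mesh, water effect, looping audio and touch cylinder collision.
; Overlapping pawns launch upward at constant speed by offsetting vertical velocity, gaining a launched effect and fall-damage immunity; hoverboards launch separately.
; Overlaps also begin and end Curie water interactions, tracked per actor in a handle map.
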(event OnComponentBeginOverlap(Touch_Cylinder)
  (params (OverlappedComponent :type Primitive%20Component%20Object%20Reference) (OtherActor :type Actor%20Object%20Reference) (OtherComp :type Primitive%20Component%20Object%20Reference) (OtherBodyIndex :type Integer) (bFromSweep :type Boolean) (SweepResult :type Hit%20Result%20Structure%20%28by%20ref%29))
  (if (Variables|FireHydrant__2|GetCan_Jump)
    (CallFunction|LaunchActor OtherActor OtherComp)
    (if (and (not (Utilities|Map|Contains (Variables|Default|GetInteractHandleMap) OtherActor)) (Curie|IsCurieActive))
      (bind _returnvalue (Curie|ExecuteBeginWaterInteract self self OtherActor 1e+02 SweepResult))
      (Utilities|Map|Add (Variables|Default|GetInteractHandleMap) OtherActor _returnvalue))))

(event OnComponentEndOverlap(Touch_Cylinder)
  (params (OverlappedComponent :type Primitive%20Component%20Object%20Reference) (OtherActor :type Actor%20Object%20Reference) (OtherComp :type Primitive%20Component%20Object%20Reference) (OtherBodyIndex :type Integer))
  (bind _interacthandlemap (Variables|Default|GetInteractHandleMap))
  (bind (_value _returnvalue) (Utilities|Map|Find _interacthandlemap OtherActor))
  (if _returnvalue
    (Curie|ExecuteEndWaterInteract OtherActor _value)
    (Utilities|Map|Remove (Variables|Default|GetInteractHandleMap) OtherActor)))

(event EventCustomStateChanged
  (CallFunction|HandleExplodedState))

(event EventOnDamageServer
  (params (Damage :type Float%20%28single-precision%29) (DamageTags :type Gameplay%20Tag%20Container%20Structure%20%28by%20ref%29) (Momentum :type Vector :default "0, 0, 0") (HitInfo :type Hit%20Result%20Structure%20%28by%20ref%29) (InstigatedBy :type Controller%20Object%20Reference) (DamageCauser :type Actor%20Object%20Reference) (EffectContext :type Gameplay%20Effect%20Context%20Handle%20Structure))
  (Building|SetCustomState "Exploded"))

(event EventOnDeathPlayEffects
  (params (Damage :type Float%20%28single-precision%29) (DamageTags :type Gameplay%20Tag%20Container%20Structure%20%28by%20ref%29) (Momentum :type Vector :default "0, 0, 0") (HitInfo :type Hit%20Result%20Structure%20%28by%20ref%29) (InstigatedBy :type Fort%20Pawn%20Object%20Reference) (DamageCauser :type Actor%20Object%20Reference) (EffectContext :type Gameplay%20Effect%20Context%20Handle%20Structure))
  (Audio|Components|Audio|FadeOut (Variables|FireHydrant__2|GetWater_Loop_Audio) 0.2 0.0 "Linear" :!class /Script/Engine.AudioComponent)
  (Audio|PlaySoundatLocation "/Game/Sounds/Fort_GamePlay_Sounds/Bouncers/FireHydrant_WaterStream_Stop_Cue.FireHydrant_WaterStream_Stop_Cue" (Transformation|GetWorldLocation (Variables|FireHydrant__2|GetWater_Loop_Audio))))

(fn HandleExplodedState
  (bind _touch_cylinder (Variables|Default|GetTouch_Cylinder))
  (bind _water_loop_audio (Variables|FireHydrant__2|GetWater_Loop_Audio))
  (bind _water_effect (Variables|FireHydrant__2|GetWater_Effect))
  (switch string (Building|GetCustomState)
    (:Default
      (Variables|FireHydrant__2|SetCan_Jump false)
      (Variables|Mesh|SetStaticMesh (Variables|Default|GetMesh_Default))
      (Components|Activation|Deactivate _water_effect)
      (Collision|SetCollisionEnabled (Variables|Default|GetTouch_Cylinder) "NoCollision"))
    (:Exploded
      (Variables|FireHydrant__2|SetCan_Jump true)
      (Variables|Mesh|SetStaticMesh (Variables|Default|GetMesh_Damaged))
      (Components|Activation|Activate _water_effect false)
      (Audio|Components|Audio|FadeIn _water_loop_audio 0.0 1.0 0.0 "Linear" :!class /Script/Engine.AudioComponent)
      (Components|Activation|Activate _water_loop_audio false)
      (Collision|SetCollisionEnabled _touch_cylinder "QueryOnly")
      (Collision|SetCollisionResponsetoAllChannels _touch_cylinder "ECR_Ignore")
      (Collision|SetCollisionResponsetoChannel _touch_cylinder "ECC_Pawn" "ECR_Overlap"))))

(fn LaunchActor
  (params (ActorToLaunch :type Actor%20Object%20Reference) (ComponentTriggeringLaunch :type Primitive%20Component%20Object%20Reference))
  (bind _water_effect (Variables|FireHydrant__2|GetWater_Effect))
  (bind _water_effect_2 (Variables|FireHydrant__2|GetWater_Effect))
  (bind _pawntolaunch (LocalVariables|GetPawnToLaunch))
  (bind _asfort_pawn (Utilities|Casting|CastToFortPawn ActorToLaunch)
    (:then
      (if (not (or (GameplayTags|HasMatchingGameplayTag (Ability|GetAbilitySystemComponent _asfort_pawn) "(TagName=\"Gameplay.Action.Player.Launched\")") (GameplayTags|HasMatchingGameplayTag (Ability|GetAbilitySystemComponent _asfort_pawn) "(TagName=\"Granted.Riding.Player.IsRiding\")")))
        (LocalVariables|SetPawnToLaunch _asfort_pawn)
        (if (Utilities|Equal(Object) (Variables|/Script/Engine.Character|GetCapsuleComponent _asfort_pawn) ComponentTriggeringLaunch)
          (Components|Activation|Deactivate _water_effect)
          (Components|Activation|Activate _water_effect false)
          (Character|LaunchCharacter _pawntolaunch (Math|Vector|MakeVector 0.0 0.0 (+ (* (.z (Transformation|GetVelocity (LocalVariables|GetPawnToLaunch))) -1) 2000)) false false)
          (GameplayEffects|ApplyGameplayEffectToSelf :self (Ability|GetAbilitySystemComponent _pawntolaunch) :GameplayEffectClass (Variables|Default|GetGE_PlayerLaunched) :Level 0.0)
          (Variables|/Script/FortniteGame.FortPawn|SetbIgnoreNextFallingDamage true _pawntolaunch)
          (Audio|Components|Audio|Play (Variables|FireHydrant__2|GetLaunch_Audio) 0.0)
          (bind _asfort_player_controller (Utilities|Casting|CastToFortPlayerController (Pawn|GetController _pawntolaunch)))
          (Game|Feedback|LocalPlayForceFeedback _asfort_player_controller (Variables|Default|GetLaunchForceFeedback) false "Hydrant"))))
    (:CastFailed
      (bind _success (Vehicles|LaunchHoverboard ActorToLaunch "0, 0,2000.000000" "Hydrant"))
      (if _success
        (Components|Activation|Deactivate _water_effect_2)
        (Components|Activation|Activate _water_effect_2 false)
        (Audio|Components|Audio|Play (Variables|FireHydrant__2|GetLaunch_Audio) 0.0)))))

(fn ConstructionScript)
