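; Fire hydrant Blueprint for a Fortnite environment, named Rufus_Hydrant01, with intact and broken meshes and a water effect.
; Its construction script copies the hydrant and broken hydrant meshes into the default and damaged mesh variables.
; It then shifts the water effect component by a configurable relative offset and calls HandleExplodedState to handle the exploded state.
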(fn ConstructionScript
  (Variables|Default|SetMesh_Default (Variables|Setup|GetHydrant_Mesh))
  (Variables|Default|SetMesh_Damaged (Variables|Setup|GetHydrant_Broken_Mesh))
  (Transformation|AddRelativeLocation (Variables|FireHydrant__2|GetWater_Effect) (Variables|Setup|GetWaterEffectOffset) false false)
  (CallFunction|HandleExplodedState))
